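; Blueprint for the SlickFiber large vat prop that plays an electric zap reaction when it takes damage.
; The damage effects event activates the Effect_Electric_ON_Hit component, referenced under PowerTransformer03b variables, and spawns a transformer zap sound attached to it.
; Damage amount, tags and hit details are unused, so every call plays the same effect, and the construction script is empty.
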(event EventOnDamagePlayEffects
  (params (Damage :type Float%20%28single-precision%29) (DamageTags :type Gameplay%20Tag%20Container%20Structure%20%28by%20ref%29) (Momentum :type Vector :default "0, 0, 0") (HitInfo :type Hit%20Result%20Structure%20%28by%20ref%29) (InstigatedBy :type Fort%20Pawn%20Object%20Reference) (DamageCauser :type Actor%20Object%20Reference) (EffectContext :type Gameplay%20Effect%20Context%20Handle%20Structure))
  (bind _effect_electric_on_hit (Variables|PowerTransformer03b|GetEffect_Electric_ON_Hit))
  (Components|Activation|Activate _effect_electric_on_hit false)
  (Audio|SpawnSoundAttached "/Game/Sounds/Fort_Traps/Wall_Electric/Transformer_Electric_Zap_Cue.Transformer_Electric_Zap_Cue" _effect_electric_on_hit))

(fn ConstructionScript)
